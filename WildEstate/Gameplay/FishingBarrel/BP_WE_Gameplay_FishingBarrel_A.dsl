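; BP_WE_Gameplay_FishingBarrel_A is the fishing barrel gameplay Blueprint for the WildEstate set.
; It reports itself as targetable by linear movers and disallows invisible props.
; Its construction script runs the parent's, then adds the tiered-chest material tag to the actor's tags so already-placed barrels receive the correct chrome material.
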(fn IsTargetableByLinearMover
  (return true))

(fn ShouldInvisiblePropsBeAllowed
  (return false))

(fn ConstructionScript
  ;[0.32 0.9402 1] Make sure actors that are already placed in levels get this tag so that they get the right chrome material added to them.
  (|CallParentFunction)
  (Utilities|Array|AddUnique (Variables|Actor|GetTags) "Lackluster.Material.TieredChestAthena"))
